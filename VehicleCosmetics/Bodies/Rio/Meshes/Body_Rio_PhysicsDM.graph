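# Physics asset for the Rio vehicle body mesh, providing a simplified collision volume for the car body.
# A single body on the root bone holds one box shape, offset upward and slightly forward to enclose the body.
# There are no constraints, so the whole body acts as one rigid piece.
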PhysicsAsset /VehicleCosmetics/Bodies/Rio/Meshes/Body_Rio_PhysicsDM
bodies {root [{shapeType=Box, extentX=444, extentY=216, extentZ=132, shapeName=root_box, center=<24 0 100>}]}
constraints []
